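; Gameplay ability for Creative melee weapons that runs block, dodge or both, as the weapon's target action selects.
; Blocking applies SetByCaller damage resistance, a montage and camera mode, ending on release or suppressing tags like swimming or sliding.
; A BlockedDamage attribute listener spends sprint-component energy per hit, applying BlockBreak when that fails.
; Dodging checks tags and acceleration relative to camera yaw, then jumps or sends the melee dodge event.
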(event Ability|EventActivateAbility
  ;[0 1 0.3366 0.25 64] Activate and input listen
  (Default|InitAbility)
  (|CallParentFunction)
  (if (Utilities|Enum|Equal(Enum) (Variables|Default|GetCurrentTargetAction) "NewEnumerator0")
    (Ability|EndAbility)
    (elif (Utilities|Enum|NotEqual(Enum) (Variables|Default|GetCurrentTargetAction) "NewEnumerator2")
      (if (Variables|Block|GetbWantsToBlock)
        (CallFunction|StartBlocking)
        (else
          _))
      (if (Variables|Block|GetbWantsToDodge)
        (CallFunction|ListenForDodging))
      (else
        (Utilities|FlowControl|DoOnce false
          (:Completed
            (CallFunction|PerformDodge)))))))

(event Ability|EventOnAbilityInputReleased
  ;[0 1 0.3366 0.25 64] Activate and input listen
  (|CallParentFunction)
  (if (Variables|Block|GetbWantsToBlock)
    (CallFunction|StopBlocking)
    (else
      (Ability|EndAbility)))
  (Utilities|FlowControl|DoOnce false
    (:Completed
      (CallFunction|PerformDodge))))

(event Custom|Listen%20for%20Dodging
  ;[0 0.7015 1 0.25 64] Dodge Activation Logic
  (Ability|Tasks|WaitGameplayEvent "(TagName=\"Event.Jump.Pressed\")"
    (:EventReceived
      (bind _newparam (CallFunction|CanPerformDodge))
      (if _newparam
        (bind _dodgedirection (CallFunction|ReturnDodgeDirection))
        (if (== _dodgedirection 0)
          (if (Utilities|Enum|NotEqual(Enum) (Variables|Default|GetCurrentTargetAction) "NewEnumerator2")
            (Character|Jump (Variables|RuntimeClassMembers|GetPlayerPawn))
            (else
              (Ability|SendGameplayEvent "(TagName=\"Event.Melee.Dodge\")")
              (Utilities|FlowControl|Delay 0.2)))
          (if (Utilities|Enum|Equal(Enum) (Variables|Default|GetCurrentTargetAction) "NewEnumerator2")
            (Ability|EndAbility))
          (else
            _))))))

(event Custom|PerformDodge
  ;[0 0.7015 1 0.25 64] Dodge Activation Logic
  (bind _newparam (CallFunction|CanPerformDodge))
  (if _newparam
    (bind _dodgedirection (CallFunction|ReturnDodgeDirection))
    (if (== _dodgedirection 0)
      (if (Utilities|Enum|NotEqual(Enum) (Variables|Default|GetCurrentTargetAction) "NewEnumerator2")
        (Character|Jump (Variables|RuntimeClassMembers|GetPlayerPawn))
        (else
          (Ability|SendGameplayEvent "(TagName=\"Event.Melee.Dodge\")")
          (Utilities|FlowControl|Delay 0.2)))
      (if (Utilities|Enum|Equal(Enum) (Variables|Default|GetCurrentTargetAction) "NewEnumerator2")
        (Ability|EndAbility))
      (else
        _))))

(event Custom|Listen%20for%20Block%20Cancel%20Events
  (bind _playerpawn (Variables|RuntimeClassMembers|GetPlayerPawn))
  ;[0.7385 0 1 0.25 64] Listens for a list of possible events that will suppress the block.
  (Utilities|FlowControl|Sequence
    (:then_0
      (Ability|Tasks|WaitGameplayTagAdd "(TagName=\"Gameplay.Action.Player.SurfaceSwimming\")" _playerpawn
        (:Added
          _)) ; End the ability when player starts swimming
    )
    (:then_1
      (Ability|Tasks|WaitGameplayTagAdd "(TagName=\"Gameplay.Action.Player.CarryingPlayer\")" _playerpawn
        (:Added
          _)) ; End the ability when player starts carrying a DBNO player
    )
    (:then_2
      (Ability|Tasks|WaitGameplayTagAdd "(TagName=\"Gameplay.Action.Player.InteractSearch\")" _playerpawn
        (:Added
          _)) ; End the ability when player starts Interact Searching
    )
    (:then_3
      (Ability|Tasks|WaitGameplayTagAdd "(TagName=\"Gameplay.Action.Player.InteractUse\")" _playerpawn
        (:Added
          _)) ; End the ability when player starts Interact Using
    )
    (:then_4
      (Ability|Tasks|WaitGameplayTagAdd "(TagName=\"Gameplay.Status.HidingProp\")" _playerpawn
        (:Added
          _)) ; End the ability when player enters a Hiding Prop
    )
    (:then_5
      (Ability|Tasks|WaitGameplayTagAdd "(TagName=\"MovementMode.Skydiving\")" _playerpawn
        (:Added
          _)) ; End the ability when player starts Skydiving
    )
    (:then_6
      (Ability|Tasks|WaitGameplayTagAdd "(TagName=\"MovementMode.Parachuting\")" _playerpawn
        (:Added
          _)) ; End the ability when player starts Skydiving
    )
    (:then_7
      (Ability|Tasks|WaitGameplayTagAdd "(TagName=\"MovementMode.Sliding\")" _playerpawn
        (:Added
          _)))
    (:then_8
      (Ability|Tasks|WaitGameplayTagAdd "(TagName=\"Event.IsUsingJetpack\")" _playerpawn
        (:Added
          _)))
    (:then_9
      (Ability|Tasks|WaitGameplayTagAdd "(TagName=\"Event.Melee.BlockBreak\")" _playerpawn
        (:Added
          (Utilities|FlowControl|Delay 0.2) ; We delay to ensure the client also gets the block damage update to consume stamina for the block attempt.
        ))))
  (Ability|EndAbility))

(event Ability|EventOnEndAbility
  (params (bWasCancelled :type Boolean))
  ;[1 0.2065 0 0.25 64] Deconstruction
  (|CallParentFunction bWasCancelled)
  (CallFunction|StopBlockingImmediately)
  (CallFunction|AbilityCleanup))

(event Custom|Stop%20Blocking%20Immediately
  ;[1 0.2065 0 0.25 64] Deconstruction
  (Camera|PopCameraMode)
  (if (== (Variables|Block|GetBlockMontage) (Animation|GetCurrentMontage))
    (Ability|Animation|MontageSetNextSectionName "Loop" "Outro")))

(event Custom|Start%20Blocking
  ;[0 0.7015 1 0.635 64] Block states
  ;;[0 1 0.5181 0.25 32] Start Blocking
  (if (ScalableFloat|GetValueasBool (Variables|Default|GetBlockEnabledHotfixRow))
    (CallFunction|ListenForBlockCancelEvents)
    (CallFunction|ToggleBlockHUDMessaging true)
    (bind _asynctaskproxy (Ability|Tasks|WaitForAttributeChangeWithComparison :InAttribute "(AttributeName=\"BlockedDamage\",Attribute=/Script/FortniteGame.FortHealthSet:BlockedDamage,AttributeOwner=\"/Script/CoreUObject.Class'/Script/FortniteGame.FortHealthSet'\")" :InComparisonType "GreaterThan" :InComparisonValue 0.0 :TriggerOnce false)
      (:then
        (Variables|Default|SetWaitForBlockedDamageAsyncTask _asynctaskproxy)
        (Ability|Tasks|PlayMontageAndWaitNotify "None" (Variables|Block|GetBlockMontage))
        (Ability|ExecuteGameplayCueOnOwner (Variables|Default|GetGCN_StartBlock) (Class|AthenaFunctionLibrary|MakeSimpleFortEffectContext (Ability|GetAvatarActorfromActorInfo) (Ability|GetAvatarActorfromActorInfo)))
        (Networking|ForceNetUpdate (Variables|RuntimeClassMembers|GetPlayerPawn)) ; Doing this because we're using a GC tag to run the anim.\nThis forces the GC through quickly for the other clients to see.
      )
      (:OnChange
        (CallFunction|ApplyBlockEffects)
        (bind _returnvalue (GameplayTags|HasMatchingGameplayTag(Message) (Variables|RuntimeClassMembers|GetPlayerPawn) "(TagName=\"Event.Melee.BlockBreak\")"))
        (if (not _returnvalue)
          (Ability|HasAuthority
            (:True
              (GameplayCue|ExecuteGameplayCue (Variables|RuntimeClassMembers|GetPlayerPawn) (Variables|Default|GetBlockedGameplayCueTag) (Class|AthenaFunctionLibrary|MakeSimpleFortEffectContext (Ability|GetAvatarActorfromActorInfo) (Ability|GetAvatarActorfromActorInfo))))))))))

(event Custom|Stop%20Blocking
  ;[0 0.7015 1 0.635 64] Block states
  ;;[1 0.3399 0 0.25 32] Stop Blocking
  (CallFunction|ToggleBlockHUDMessaging false)
  (Development|Editor|StackTrace)
  (if (== (Variables|Block|GetBlockMontage) (Animation|GetCurrentMontage))
    (bind _asynctaskproxy (Ability|Tasks|SetNextMontageSectionandWait (Variables|Block|GetBlockMontage))
      (:then
        (Variables|AsyncTaskHandles|SetBlockMontageSectionWaitAsyncTask _asynctaskproxy)
        (Ability|ExecuteGameplayCueOnOwner (Variables|Default|GetGCN_EndBlock) (Class|AthenaFunctionLibrary|MakeSimpleFortEffectContext (Ability|GetAvatarActorfromActorInfo) (Ability|GetAvatarActorfromActorInfo)))
        (Ability|EndAbility))
      (:OnComplete
        _)
      (:OnInterrupted
        _))
    (else
      _))
  (CallFunction|StopBlockingImmediately))

(fn AbilityCleanup
  (Ability|RemoveGameplayEffectFromOwnerWithHandle (Variables|Default|GetBlockingDamageResistanceHandle))
  (Development|PrintString "Removed Block" true true "(R=1.000000,G=0.015037,B=0.000000,A=1.000000)")
  (Camera|PopCameraMode)
  (bind _waitforblockeddamageasynctask (Variables|Default|GetWaitForBlockedDamageAsyncTask)
    (:then
      (GameplayTasks|EndTask _waitforblockeddamageasynctask))
    (:else
      _))
  (Blocking|RemoveTagstoBlockwithMelee(Message) (Variables|RuntimeClassMembers|GetPlayerPawn) (Variables|Default|GetAttacksToBlock))
  (if (GameplayTags|HasMatchingGameplayTag (Variables|RuntimeClassMembers|GetPlayerPawn) "(TagName=\"Ability.Melee.Block\")")
    (Ability|RemoveGameplayEffectFromOwnerWithGrantedTags "(GameplayTags=((TagName=\"Ability.Melee.Block\")))")))

(fn ApplyBlockEffects
  (Variables|Default|GetEnergyComp
    (:then
      (bind (_primary_attack_cost _secondary_attack_cost _block_hit_cost) (CreativeMeleeWeaponInterface|GetEnergyCosts(Message) (Variables|RuntimeClassMembers|GetBP_WeaponActor) false))
      (bind (_outamountofenergyused _returnvalue) (Energy|UseEnergy (Variables|Default|GetEnergyComp) _block_hit_cost true 0))
      (Ability|HasAuthority
        (:True
          (if (not _returnvalue)
            (Development|PrintString "FAILED TO BLOCK ATTACK")
            (Ability|ApplyGameplayEffectToOwner "/CRD_Melee/Weapons/GE/GE_BlockBreak.GE_BlockBreak_C")))))
    (:else)))

(fn ApplyParryEffects
  (Ability|ApplyGameplayEffectToOwner (Variables|Default|GetGE%20Parry)))

(fn AsyncListenTos)

(fn CanPerformDodge
  (bind _bwantstododge (Variables|Block|GetbWantsToDodge))
  (bind _playerpawn (Variables|RuntimeClassMembers|GetPlayerPawn)
    (:then
      (return (and (not (GameplayTags|HasAnyMatchingGameplayTags _playerpawn (Variables|Block|GetDodgeListenToTags))) _bwantstododge)))
    (:else
      (return _bwantstododge))))

(fn EndBlockCleanup
  (Ability|RemoveGameplayEffectFromOwnerWithHandle (Variables|RuntimeClassMembers|GetInitializedBlockEffectHandle))
  (Ability|RemoveGameplayEffectFromOwnerWithGrantedTags "(GameplayTags=((TagName=\"Ability.Melee.Block\")))"))

(fn GetSuppressionTags
  (bind _output_get (LocalVariables|SetTags (Utilities|Array|MakeArray (Variables|Default|GetMelee%20Tag))))
  (Utilities|Array|AppendArray _output_get (GameplayTags|BreakGameplayTagContainer (Variables|Block|GetBlockListenToTags)))
  (return (GameplayTags|MakeGameplayTagContainerfromArray (LocalVariables|GetTags))))

(fn GetCurrentValidWeapon
  (bind _returnvalue (Ability|GetCurrentSourceWeapon))
  (Utilities|IsValid _returnvalue
    (:Is%20Valid
      (return _returnvalue))
    (:Is%20Not%20Valid
      (bind _outfortplayerpawn (CallFunction|GetAvatarasValidFortPlayerPawn)
        (:ValidFortPlayerPawn
          (return (Variables|/Script/FortniteGame.FortPawn|GetCurrentWeapon _outfortplayerpawn)))
        (:AvatarCastFailed
          (return 0))))))

(fn InitAbility
  (bind _returnvalue (Ability|CommitAbility))
  (if _returnvalue
    (bind _outfortplayerpawn (CallFunction|GetAvatarasValidFortPlayerPawn)
      (:ValidFortPlayerPawn
        (Development|PrintString (select (GameplayTags|HasAnyMatchingGameplayTags _outfortplayerpawn (Variables|Block|GetBlockListenToTags)) "HAS BLOCKED TAGS" "NO BLOCKED TAGS") true true "(R=1.000000,G=0.000000,B=0.940979,A=1.000000)")
        (bind _output_get (Variables|RuntimeClassMembers|SetPlayerPawn _outfortplayerpawn))
        (Variables|RuntimeClassMembers|SetBP_WeaponActor (Ability|GetCurrentSourceWeapon))
        (Variables|Default|SetEnergyComp (Actor|GetComponentbyClass _output_get "/TacticalSprintGame/Gameplay/TacticalSprintEnergyComponent.TacticalSprintEnergyComponent_C"))
        (CallFunction|SetCurrentTargetAction)
        (Development|PrintString "Added Block" true true "(R=0.000000,G=1.000000,B=0.161746,A=1.000000)")
        (Camera|PushCameraMode (Variables|Default|GetCamera%20Mode))
        (if (Variables|Block|GetbWantsToBlock)
          (Blocking|AddTagstoBlockwithMelee(Message) (Variables|RuntimeClassMembers|GetPlayerPawn) (Variables|Default|GetAttacksToBlock))
          (bind _damageresistancepercentage (CreativeMeleeWeaponInterface|GetBlockData(Message) (Variables|RuntimeClassMembers|GetBP_WeaponActor)))
          (bind _returnvalue_2 (Ability|GameplayEffect|AssignTagSetbyCallerMagnitude (Ability|MakeOutgoingGameplayEffectSpec (Variables|Default|GetGE%20Block)) "(TagName=\"SetByCaller.MeleeManager.DamageResistance\")" (Math|Float|Clamp(Float) (* _damageresistancepercentage 0.01))))
          (bind _returnvalue_3 (Ability|ApplyGameplayEffectSpecToOwner _returnvalue_2))
          (Variables|Default|SetBlockingDamageResistanceHandle _returnvalue_3))))
    (else
      (Ability|EndAbility))))

(fn IsButtonToggleEnabled
  (return (or (TouchPlatform|IsUsingTouch (Fortnite|GetFortPlayerControllerfromActor 0)) (Variables|Default|GetbUserOptionToggle))))

(fn CanActivateAbility
  (params (ActorInfo :type Gameplay%20Ability%20Actor%20Info%20Structure) (Handle :type Gameplay%20Ability%20Spec%20Handle%20Structure))
  (bind (_owneractor _avataractor) (Utilities|Struct|BreakGameplayAbilityActorInfo ActorInfo))
  (bind _asfort_pawn (Utilities|Casting|CastToFortPawn _avataractor)
    (:then
      (if (not (GameplayTags|HasMatchingGameplayTag _asfort_pawn (GameplayTags|MakeLiteralGameplayTag "(TagName=\"Event.Melee.BlockBreak\")")))
        (return 0 true)
        (else
          (return 0 false))))
    (:CastFailed
      (return 0 true))))

(fn PlayerCanPerformBlock
  ;[1 1 1] Player can only block if this query is true
  (if (ScalableFloat|GetValueasBool (Variables|Default|GetBlockEnabledHotfixRow))
    (bind _playerpawn (Variables|RuntimeClassMembers|GetPlayerPawn)
      (:then
        (if (GameplayTags|HasMatchingGameplayTag _playerpawn (Variables|Default|GetMelee%20Tag))
          (if (not (GameplayTags|HasMatchingGameplayTag _playerpawn (Variables|Default|GetAllowActioninMeleeTag)))
            (return false)
            (else
              _)))
        (return (not (GameplayTags|HasAnyMatchingGameplayTags _playerpawn (Variables|Block|GetBlockListenToTags)))))
      (:else))
    (else
      (return false))))

(fn ReturnDodgeDirection
  (bind _playerpawn (Variables|RuntimeClassMembers|GetPlayerPawn))
  (bind _returnvalue (Pawn|GetController _playerpawn))
  (bind (_returnvalue_roll _returnvalue_pitch _returnvalue_yaw) (Math|Rotator|BreakRotator (Pawn|GetControlRotation _returnvalue :!class /Script/Engine.Controller)))
  (bind _returnvalue_2 (Math|Vector|GetForwardVector (Math|Rotator|MakeRotator _returnvalue_roll 0.0 _returnvalue_yaw)))
  (bind _returnvalue_3 (Math|Vector|GetRightVector (Math|Rotator|MakeRotator _returnvalue_roll 0.0 _returnvalue_yaw)))
  (bind _charactermovement (Variables|/Script/Engine.Character|GetCharacterMovement _playerpawn))
  (bind _returnvalue_4 (Pawn|Components|CharacterMovement|GetCurrentAcceleration _charactermovement))
  (bind _returnvalue_5 (Math|Vector|Normalize _returnvalue_4))
  (bind _returnvalue_6 (Math|Vector|DotProduct _returnvalue_5 _returnvalue_2))
  (bind _returnvalue_7 (Math|Vector|DotProduct _returnvalue_5 _returnvalue_3))
  ;[1 1 1] GA_Melee_Dodge_Parent has you going backwards if current acceleration is 0. However, we want a jump in-place, so we're going to use "North" because that will jump in our case.
  (if (Math|Vector|VectorIsNearlyZero _returnvalue_4)
    _
    (elif (Math|Float|NearlyEqual(Float) _returnvalue_6 1.0 0.35)
      _
      (elif (Math|Float|NearlyEqual(Float) _returnvalue_6 -1.0 0.35)
        (LocalVariables|SetLOCAL_Direction 1) ; South
        (return (LocalVariables|GetLOCAL_Direction))
        (elif (Math|Float|NearlyEqual(Float) _returnvalue_7 1.0 0.5)
          (LocalVariables|SetLOCAL_Direction 2) ; East
          (elif (Math|Float|NearlyEqual(Float) _returnvalue_7 -1.0 0.5)
            (LocalVariables|SetLOCAL_Direction 3) ; West
            (else
              (LocalVariables|SetLOCAL_Direction 1) ; South
            ))))))
  (LocalVariables|SetLOCAL_Direction 0) ; North
)

(fn SendFailMessage
  ;[1 0.01591 0 0.25] If we can't block because we don't have enough energy, say that. Else just use a generic message.
  (bind _playerpawn (Variables|RuntimeClassMembers|GetPlayerPawn)
    (:then
      (if (GameplayTags|HasMatchingGameplayTag _playerpawn "(TagName=\"Ability.Melee.NotEnoughEnergy.Block\")")
        (FailedAction|UnabletoPerformAction _playerpawn true (select (GameplayTags|HasMatchingGameplayTag _playerpawn "(TagName=\"Ability.Melee.NotEnoughEnergy.Block\")") (Variables|Default|GetFailed%20Reason%20Energy) (Variables|Default|GetFailed%20Reason%20Generic)))
        (GameplayCue|ExecuteGameplayCue _playerpawn (Variables|Default|GetGCN_NotEnoughEnergy) (Class|AthenaFunctionLibrary|MakeSimpleFortEffectContext _playerpawn))))
    (:else)))

(fn SetCurrentTargetAction
  (bind _ascreative_melee_parent (Utilities|Casting|CastToCreative_Melee_Parent (Ability|GetCurrentSourceWeapon)))
  (bind _targetaction (CreativeMeleeWeaponInterface|GetTargetActionSelection _ascreative_melee_parent :!class /CRD_Melee/Weapons/Creative_Melee_Parent.Creative_Melee_Parent_C))
  (Variables|Default|SetCurrentTargetAction _targetaction)
  (Variables|Block|SetbWantsToBlock (or (Utilities|Enum|Equal(Enum) (Variables|Default|GetCurrentTargetAction) "NewEnumerator1") (Utilities|Enum|Equal(Enum) (Variables|Default|GetCurrentTargetAction) "NewEnumerator3")))
  (Variables|Block|SetbWantsToDodge (or (Utilities|Enum|Equal(Enum) (Variables|Default|GetCurrentTargetAction) "NewEnumerator1") (Utilities|Enum|Equal(Enum) (Variables|Default|GetCurrentTargetAction) "NewEnumerator2"))))

(fn ToggleBlockHUDMessaging
  (params (Activate :type Boolean))
  (bind _playerpawn (Variables|RuntimeClassMembers|GetPlayerPawn)
    (:then
      (bind _ascreative_melee_parent (Utilities|Casting|CastToCreative_Melee_Parent (Weapon|GetCurrentWeapon _playerpawn)))
      (Class|CreativeMeleeParent|ToggleEvadeHUDMessaging _ascreative_melee_parent Activate))
    (:else)))
