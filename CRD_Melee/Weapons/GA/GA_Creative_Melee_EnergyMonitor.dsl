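; Gameplay ability for Creative melee weapons that mirrors the player's energy level into per-action "NotEnoughEnergy" tags.
; On activation it finds the pawn's TacticalSprint energy component, refreshes once, then listens for energy changes.
; Each update passes the energy percentage to the source weapon's GetCanAttack, adding or removing loose tags for primary, secondary, block and dodge.
; Ending the ability unbinds the listener and clears all four tags.
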(event Ability|EventActivateAbility
  (bind _outfortplayerpawn (CallFunction|GetAvatarasValidFortPlayerPawn)
    (:ValidFortPlayerPawn
      (bind _output_get (Variables|Default|SetEnergyComp (Actor|GetComponentbyClass _outfortplayerpawn "/TacticalSprintGame/Gameplay/TacticalSprintEnergyComponent.TacticalSprintEnergyComponent_C")))
      (Default|UpdateAbilityEnergy _output_get)
      (Events|BindEventtoOnCurrentEnergyChanged (Variables|Default|GetEnergyComp) (delegate Custom|Energy%20updated)))
    (:AvatarCastFailed
      (Ability|EndAbility))))

(event Custom|Energy%20updated
  (params (EnergyComponent :type Fort%20Component%20Energy%20Object%20Reference) (PreviousValue :type Float%20%28single-precision%29))
  (Default|UpdateAbilityEnergy EnergyComponent))

(event Ability|EventFailedToActivatePassiveAbility
  (params (ActorInfo :type Gameplay%20Ability%20Actor%20Info%20Structure)))

(event Ability|EventOnEndAbility
  (params (bWasCancelled :type Boolean))
  (Events|UnbindEventfromOnCurrentEnergyChanged (Variables|Default|GetEnergyComp) (delegate Custom|Energy%20updated))
  (Default|ForceCleanupEnergyTags))

(fn AddOrRemoveLooseTag
  (params (Tag :type Gameplay%20Tag%20Structure) (Remove :type Boolean))
  (bind _returnvalue (GameplayTags|MakeGameplayTagContainerfromTag Tag))
  (bind _outfortplayerpawn (CallFunction|GetAvatarasValidFortPlayerPawn)
    (:ValidFortPlayerPawn
      (if Remove
        (Ability|GameplayEffect|RemoveLooseGameplayTags _outfortplayerpawn _returnvalue)
        (elif (not (GameplayTags|HasMatchingGameplayTag _outfortplayerpawn Tag))
          (Ability|GameplayEffect|AddLooseGameplayTags _outfortplayerpawn _returnvalue))))))

(fn ForceCleanupEnergyTags
  (Default|AddorRemoveLooseTag "(TagName=\"Ability.Melee.NotEnoughEnergy.Primary\")" true)
  (Default|AddorRemoveLooseTag "(TagName=\"Ability.Melee.NotEnoughEnergy.Secondary\")" true)
  (Default|AddorRemoveLooseTag "(TagName=\"Ability.Melee.NotEnoughEnergy.Block\")" true)
  (Default|AddorRemoveLooseTag "(TagName=\"Ability.Melee.NotEnoughEnergy.Dodge\")" true))

(fn UpdateAbilityEnergy
  (params (EnergyComp :type Fort%20Component%20Energy%20Object%20Reference))
  (bind _returnvalue (Energy|GetCurrentEnergyPercentage EnergyComp))
  (bind _ascreative_melee_parent (Utilities|Casting|CastToCreative_Melee_Parent (Ability|GetCurrentSourceWeapon)))
  (Default|AddorRemoveLooseTag "(TagName=\"Ability.Melee.NotEnoughEnergy.Primary\")" (Class|CreativeMeleeParent|GetCanAttack _ascreative_melee_parent _returnvalue))
  (Default|AddorRemoveLooseTag "(TagName=\"Ability.Melee.NotEnoughEnergy.Secondary\")" (.CanSecondary (Class|CreativeMeleeParent|GetCanAttack _ascreative_melee_parent _returnvalue)))
  (Default|AddorRemoveLooseTag "(TagName=\"Ability.Melee.NotEnoughEnergy.Block\")" (.CanBlock (Class|CreativeMeleeParent|GetCanAttack _ascreative_melee_parent _returnvalue)))
  (Default|AddorRemoveLooseTag "(TagName=\"Ability.Melee.NotEnoughEnergy.Dodge\")" (.CanDodge (Class|CreativeMeleeParent|GetCanAttack _ascreative_melee_parent _returnvalue))))
