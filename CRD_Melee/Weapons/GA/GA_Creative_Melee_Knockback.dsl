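; Gameplay ability that knocks back its avatar from a Creative melee weapon hit, firing a knockback gameplay cue on activation.
; FindLauncher passes the launcher to GenerateValuesFromLauncher, which flags AI avatars by tag and checks the weapon owner for the Mantis jump-attack tag.
; The weapon returns knockback and knockup power for those flags through CreativeMeleeWeaponInterface.
; GetInitialKnockMag scales the normalized horizontal direction by knockback power and uses knockup power as Z.
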(event Ability|EventActivateAbility
  (|CallParentFunction)
  (Ability|ExecuteGameplayCueOnOwner (Variables|Default|GetGCN_Knockback)))

(event Ability|EventOnEndAbility
  (params (bWasCancelled :type Boolean)))

(event Ability|EventFailedToActivatePassiveAbility
  (params (ActorInfo :type Gameplay%20Ability%20Actor%20Info%20Structure))
  (|CallParentFunction))

(fn FindLauncher
  (params (DefaultLauncher :type Actor%20Object%20Reference))
  (Utilities|FlowControl|Sequence
    (:then_0
      (CallFunction|GenerateValuesFromLauncher DefaultLauncher))
    (:then_1
      (return DefaultLauncher))))

(fn GenerateValuesfromLauncher
  (params (Launcher%20Actor :type Actor%20Object%20Reference))
  (bind _returnvalue (Ability|GetAvatarActorfromActorInfo))
  (bind _ascreative_melee_parent (Utilities|Casting|CastToCreative_Melee_Parent (Variables|GetLauncher%20Actor)))
  (bind _asfort_player_pawn (Utilities|Casting|CastToFortPlayerPawn (Actor|GetOwner _ascreative_melee_parent)))
  (Utilities|IsValid _returnvalue
    (:Is%20Valid
      (bind _returnvalue_2 (GameplayTags|HasAnyMatchingGameplayTags(Message) _returnvalue (Variables|Default|GetAIGameplayTags)))
      (Variables|Default|SetAITarget _returnvalue_2))
    (:Is%20Not%20Valid
      _))
  (bind (_wantstoknock _knockback_power _knockup_power) (CreativeMeleeWeaponInterface|GetKnockbackData(Message) (Variables|GetLauncher%20Actor) (GameplayTags|HasAllTags (GameplayTags|GetOwnedGameplayTags (Ability|GetAbilitySystemComponent _asfort_player_pawn) :!class /Script/GameplayTags.GameplayTagAssetInterface) (GameplayTags|MakeLiteralGameplayTagContainer "(GameplayTags=((TagName=\"Event.Melee.Mantis.Creative.JumpAttackApplied\")))") true) (Variables|Default|GetAITarget)))
  (Variables|Default|SetKnockback%20Power _knockback_power)
  (Variables|Default|SetKnockup%20Power _knockup_power))

(fn GetInitialKnockMag
  (params (Direction_X :type Float%20%28double-precision%29) (Direction_Y :type Float%20%28double-precision%29) (Direction_Z :type Float%20%28double-precision%29) (OverrideXYMag :type Boolean) (XYMAgOverride :type Float%20%28double-precision%29))
  (bind (_returnvalue_x _returnvalue_y) (Math|Vector2D|BreakVector2D (Math|Vector2D|Normalize2D (Math|Vector2D|MakeVector2D Direction_X Direction_Y))))
  (return (Math|Vector|MakeVector (* _returnvalue_x (Variables|Default|GetKnockback%20Power)) (* _returnvalue_y (Variables|Default|GetKnockback%20Power)) (Variables|Default|GetKnockup%20Power))))
