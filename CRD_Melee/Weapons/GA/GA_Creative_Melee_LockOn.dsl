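; Gameplay ability for Creative melee weapons that locks onto a target while secondary is held, gated by hotfix and weapon option.
; Async targeting picks the actor closest to aim, then SetTargetLock drives reticle, sounds, look-input blocking and Mantis lock-on state.
; A next-tick loop lerps control rotation toward the target.
; Release, target death, range or blocking tags clear the lock; custom-camera mode sends a targeting event instead.
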(event Ability|EventActivateAbility
  (bind _findtargetspreset (Variables|Default|GetFindTargetsPreset))
  ;[1 1 1 1 64] Activate
  (if (ScalableFloat|GetValueasBool (Variables|Default|GetLockOnEnabledHotfixRow))
    (|CallParentFunction)
    (bind _outfortplayerpawn (CallFunction|GetAvatarasValidFortPlayerPawn)
      (:ValidFortPlayerPawn
        (bind _ascreative_melee_parent (Utilities|Casting|CastToCreative_Melee_Parent (Variables|/Script/FortniteGame.FortPawn|GetCurrentWeapon _outfortplayerpawn))
          (:then
            (if (Utilities|Enum|Equal(Enum) (.LockOn_89_D7A1968E484421556A3A1F8484F7BD0B (Utilities|Struct|BreakCreativeMeleeOptions (Variables|/CRD_Melee/Weapons/Creative_Melee_Parent.Creative_Melee_Parent_C|GetCachedState _ascreative_melee_parent))) "NewEnumerator0")
              (Utilities|FlowControl|Sequence
                (:then_0
                  (Ability|Tasks|WaitGameplayEvent "(TagName=\"Event.Secondary.Pressed\")"
                    (:EventReceived
                      (Utilities|IsValid _findtargetspreset
                        (:Is%20Valid
                          (if (CallFunction|IsUsingCustomCamera)
                            (Ability|SendGameplayEvent "(TagName=\"Event.Melee.Targeting.Active\")")
                            (else
                              (bind (_asynctaskproxy _targetinghandle) (Class|AsyncActionPerformTargeting|PerformTargetingAsyncAction (CallFunction|GetAvatarasFortPlayerPawn) _findtargetspreset true)
                                (:Targeted
                                  (if (Utilities|Array|IsNotEmpty (TargetingSystem|TargetingResults|GetTargetingResultsActors "/Script/TargetingSystem.Default__TargetingSubsystem" _targetinghandle))
                                    (bind _result (CallFunction|GetClosestActorByAim (TargetingSystem|TargetingResults|GetTargetingResultsActors "/Script/TargetingSystem.Default__TargetingSubsystem" _targetinghandle)))
                                    (Utilities|IsValid (Ability|GetAbilitySystemComponent _result)
                                      (:Is%20Valid
                                        ;[1 1 1 1 64] Update View
                                        (if (GameplayTags|HasAnyMatchingGameplayTags (Ability|GetAbilitySystemComponent _result) (Variables|Default|GetTagsRequiredForLockOn))
                                          (CallFunction|SetTargetLock _result)
                                          (CallFunction|UpdateLockonRotation)
                                          (Utilities|Time|SetTimerforNextTickbyEvent (delegate Custom|NextUpdate))
                                          (else
                                            _)))
                                      (:Is%20Not%20Valid
                                        _))
                                    (CallFunction|ClearTargetLock)
                                    (else
                                      (CallFunction|ClearTargetLock))))))))))))
                (:then_1
                  (Ability|Tasks|WaitGameplayEvent "(TagName=\"Event.Secondary.Released\")"
                    (:EventReceived
                      (CallFunction|ClearTargetLock)
                      (if (CallFunction|IsUsingCustomCamera)
                        (Ability|SendGameplayEvent "(TagName=\"Event.Melee.Targeting.Active\")"))))))
              (else
                (Ability|EndAbility))))
          (:CastFailed
            (Ability|EndAbility))))
      (:AvatarCastFailed
        (Ability|EndAbility)))
    (else
      (Ability|EndAbility))))

(event Custom|NextUpdate
  ;[1 1 1 1 64] Update View
  (if (Variables|Default|GetlockedOn)
    (CallFunction|UpdateLockonRotation)
    (Utilities|Time|SetTimerforNextTickbyEvent (delegate Custom|NextUpdate))))

(event Ability|EventFailedToActivatePassiveAbility
  (params (ActorInfo :type Gameplay%20Ability%20Actor%20Info%20Structure)))

(event Custom|ListenForDeath
  (params (Actor :type Actor%20Object%20Reference))
  ;[1 1 1] Clear lock when target dies
  (bind _asfort_pawn (Utilities|Casting|CastToFortPawn Actor))
  (EventDispatchers|BindEventtoOnDeathEffects _asfort_pawn (delegate Custom|LockOnTargetDied)))

(event Ability|EventOnEndAbility
  (params (bWasCancelled :type Boolean))
  (bind _lockonreticleactor (Variables|Default|GetLockOnReticleActor))
  ;[1 1 1 1 64] End Ability
  (|CallParentFunction bWasCancelled)
  (if (CallFunction|IsUsingCustomCamera)
    (Ability|SendGameplayEvent "(TagName=\"Event.Melee.Targeting.Active\")")
    (else
      _))
  (CallFunction|ClearTargetLock)
  (Utilities|IsValid _lockonreticleactor
    (:Is%20Valid
      (Class|CreativeMeleeLockonPlacementActor|SetReticleVisible _lockonreticleactor false)
      (Actor|SetLifeSpan _lockonreticleactor 0.25)
      (Variables|Default|SetLockOnReticleActor 0))))

(event Custom|LockOnTargetDied
  (params (DeadActor :type Actor%20Object%20Reference) (Damage :type Float%20%28single-precision%29) (InstigatedBy :type Fort%20Pawn%20Object%20Reference) (HitLocation :type Vector) (DamageCauser :type Actor%20Object%20Reference))
  ;[1 1 1] Clear lock when target dies
  (CallFunction|ClearTargetLock))

(event Custom|StopListeningForDeath
  (params (Actor :type Actor%20Object%20Reference))
  ;[1 1 1] Clear lock when target dies
  (Utilities|Casting|CastToFortPawn Actor)
  (EventDispatchers|UnbindEventfromOnDeathEffects (stale Custom|ListenForDeath _asfort_pawn) (delegate Custom|LockOnTargetDied)))

(fn ClearTargetLock
  (CallFunction|SetTargetLock 0))

(fn GetClosestActorbyAim
  (params (Actors :type Array%20of%20Actor%20Object%20References))
  (bind _aimvector (LocalVariables|GetAimVector))
  (bind _returnvalue (Math|Vector|Normalize _aimvector))
  (bind _aimorigin (LocalVariables|GetAimOrigin))
  (bind _outfortplayerpawn (CallFunction|GetAvatarasValidFortPlayerPawn)
    (:ValidFortPlayerPawn
      (LocalVariables|SetAimVector (Math|Vector|GetForwardVector (Pawn|GetBaseAimRotation _outfortplayerpawn)))
      (LocalVariables|SetAimOrigin (Camera|GetCameraLocation (Variables|/Script/Engine.PlayerController|GetPlayerCameraManager (Fortnite|GetFortPlayerControllerfromActor _outfortplayerpawn))))
      (for _array_element (Variables|GetActors)
        (if (> (Math|Vector|DotProduct _returnvalue (Math|Vector|Normalize (- (Transformation|GetActorLocation _array_element) _aimorigin))) (LocalVariables|GetClosestDotProductAngle))
          (LocalVariables|SetClosestDotProductAngle (Math|Vector|DotProduct _returnvalue (Math|Vector|Normalize (- (Transformation|GetActorLocation _array_element) _aimorigin))))
          (LocalVariables|SetClosestActor _array_element)))
      (return (LocalVariables|GetClosestActor)))))

(fn GetLockOnWidgetActor
  (if (Utilities|IsValid (Variables|Default|GetLockOnReticleActor) :!class /Script/Engine.KismetSystemLibrary)
    (return (Variables|Default|GetLockOnReticleActor))
    (else
      (bind _outfortplayerpawn (CallFunction|GetAvatarasValidFortPlayerPawn)
        (:ValidFortPlayerPawn
          (bind _returnvalue (Game|SpawnActorfromClass "/CRD_Melee/UI/CreativeMelee_LockOnPlacementActor.CreativeMelee_LockOnPlacementActor_C" (Transformation|GetActorTransform _outfortplayerpawn) "AlwaysSpawn" "OverrideRootScale" (Fortnite|GetControllerfromActor _outfortplayerpawn) 0))
          (bind _output_get (Variables|Default|SetLockOnReticleActor _returnvalue))
          (return _output_get))))))

(fn IsFacingTarget
  (params (LockTarget :type Actor%20Object%20Reference))
  (bind _outfortplayerpawn (CallFunction|GetAvatarasValidFortPlayerPawn)
    (:ValidFortPlayerPawn
      (return (>= (Math|Vector|DotProduct (Math|Vector|GetForwardVector (Pawn|GetControlRotation _outfortplayerpawn :!class /Script/Engine.Pawn)) (Math|Vector|Normalize (- (Transformation|GetActorLocation LockTarget) (Transformation|GetActorLocation _outfortplayerpawn)))) (Variables|Default|GetFacingTolerance))))
    (:AvatarCastFailed
      (return false))))

(fn IsTargetInRange
  (params (LockedTarget :type Actor%20Object%20Reference))
  (bind _outfortplayerpawn (CallFunction|GetAvatarasValidFortPlayerPawn)
    (:ValidFortPlayerPawn
      (return (<= (Transformation|GetDistanceTo LockedTarget _outfortplayerpawn) (Variables|Default|GetMaxTargetDistance))))
    (:AvatarCastFailed
      (return false))))

(fn IsUsingCustomCamera
  (return (GameplayTags|HasAnyMatchingGameplayTags (CallFunction|GetAvatarasFortPlayerPawn) (Variables|Default|GetCustomCameraControlsTags))))

(fn SetTargetLock
  (params (ActorToLockOnto :type Actor%20Object%20Reference))
  (bind _avatar (LocalVariables|GetAvatar))
  (bind _returnvalue (Pawn|GetController _avatar))
  (bind _returnvalue_2 (Actor|GetComponentbyClass _avatar "/Script/MantisRuntime.FortMantisPawnComponent"))
  (bind _lockonreticleactor (Variables|Default|GetLockOnReticleActor))
  (if (!= ActorToLockOnto (Variables|Default|GetlockedTarget))
    (if (Utilities|IsValid ActorToLockOnto :!class /Script/Engine.KismetSystemLibrary)
      (CallFunction|ListenforDeath ActorToLockOnto)
      (else
        (CallFunction|StopListeningforDeath (Variables|Default|GetlockedTarget))))
    (bind _output_get (Variables|Default|SetlockedTarget ActorToLockOnto))
    (Variables|Default|SetlockedOn (Utilities|IsValid _output_get :!class /Script/Engine.KismetSystemLibrary))
    (bind _outfortplayerpawn (CallFunction|GetAvatarasValidFortPlayerPawn)
      (:ValidFortPlayerPawn
        (LocalVariables|SetAvatar _outfortplayerpawn)
        (bind _actor (CallFunction|GetLockonWidgetActor))
        (Class|CreativeMeleeLockonPlacementActor|SetReticleVisible _actor (Variables|Default|GetlockedOn))
        (Audio|PlaySound2D (select (Variables|Default|GetlockedOn) (Variables|Default|GetLockOnSoundCue) (Variables|Default|GetDisengageSoundCue)))
        (CallFunction|GetAvatarasValidFortPlayerPawn
          (:ValidFortPlayerPawn
            (if (Variables|Default|GetlockedOn)
              (Input|SetIgnoreLookInput _returnvalue true)
              (bind _output_get_2 (LocalVariables|SetAvatarMantisComponent _returnvalue_2))
              (Utilities|IsValid _output_get_2
                (:Is%20Valid
                  (Class|FortMantisPawnComponent|SetLockonState _output_get_2 true (Variables|Default|GetlockedTarget))
                  (bind _ascharacter (Utilities|Casting|CastToCharacter (Variables|Default|GetlockedTarget))
                    (:then
                      (if (Transformation|DoesSocketExist (Variables|/Script/Engine.Character|GetMesh _ascharacter) "FX_Chest")
                        ;[1 1 1] Snap to chest socket
                        (Transformation|AttachActorToComponent _lockonreticleactor (Variables|/Script/Engine.Character|GetMesh _ascharacter) "FX_Chest" "SnapToTarget" "SnapToTarget" "KeepWorld" true)
                        (Transformation|SetActorRelativeLocation (Variables|Default|GetLockOnReticleActor) "0, 0, 0" false false)
                        (else
                          ;[1 1 1] 75% of capsule height
                          (Transformation|AttachActorToComponent _lockonreticleactor (Variables|/Script/Engine.Character|GetMesh _ascharacter) "" "SnapToTarget" "SnapToTarget" "KeepWorld" true)
                          (Transformation|SetActorRelativeLocation (Variables|Default|GetLockOnReticleActor) (Math|Vector|MakeVector 0.0 0.0 (Components|Capsule|GetScaledCapsuleHalfHeight (Variables|/Script/Engine.Character|GetCapsuleComponent _ascharacter))) false false) ; Move up 75% of capsule height
                        )))
                    (:CastFailed
                      (Transformation|AttachActorToActor (Variables|Default|GetLockOnReticleActor) (Variables|Default|GetlockedTarget) "" "SnapToTarget" "SnapToTarget" "KeepWorld" true)
                      ;[1 1 1] 75% of mesh height
                      (bind _boundingboxextents (Class|AthenaFunctionLibrary|GetActorMeshExtents (Variables|Default|GetlockedTarget)))
                      (Transformation|SetActorRelativeLocation (Variables|Default|GetLockOnReticleActor) (Math|Vector|MakeVector 0.0 0.0 (* 0.5 (.z _boundingboxextents))) false false)))))
              (else
                (Input|ResetIgnoreLookInput _returnvalue)
                (Transformation|DetachFromActor (Variables|Default|GetLockOnReticleActor) "KeepWorld" "KeepWorld" "KeepWorld")
                (bind _output_get_3 (LocalVariables|SetAvatarMantisComponent _returnvalue_2))
                (Utilities|IsValid _output_get_3
                  (:Is%20Valid
                    (Class|FortMantisPawnComponent|SetLockonState _output_get_3 false 0)))))))))))

(fn UpdateLockOnRotation
  (bind _pawn (LocalVariables|GetPawn))
  (bind _returnvalue (Pawn|GetController _pawn))
  (bind (_ascharacter _bsuccess) (Utilities|Casting|CastToCharacter (Variables|Default|GetlockedTarget)))
  ;[1 1 1] Valid target / locked on?
  (if (and (Variables|Default|GetlockedOn) (Utilities|IsValid (Variables|Default|GetlockedTarget) :!class /Script/Engine.KismetSystemLibrary))
    (if (not (CallFunction|IsUsingCustomCamera))
      ;[1 1 1] Distance check
      (bind _outfortplayerpawn (CallFunction|GetAvatarasValidFortPlayerPawn)
        (:ValidFortPlayerPawn
          (LocalVariables|SetPawn _outfortplayerpawn)
          (if (CallFunction|IsTargetinRange (Variables|Default|GetlockedTarget))
            ;[1 1 1] Tag check
            (bind _returnvalue_2 (GameplayTags|GetOwnedGameplayTags (Variables|Default|GetlockedTarget) :!class /Script/GameplayTags.GameplayTagAssetInterface))
            (if (not (GameplayTags|DoesContainerMatchTagQuery _returnvalue_2 (Variables|Default|GetTagsThatPreventLockOn)))
              (Fortnite|KeepCGPathOptimizations _returnvalue 21 0.2)
              (if (not (GameplayTags|HasMatchingGameplayTag (Ability|GetAbilitySystemComponent _returnvalue) "(TagName=\"MovementModeExt.ThirdPerson\")"))
                (Pawn|SetControlRotation _returnvalue (Math|Rotator|MakeRotator 0.0 (.pitch (Math|Rotator|Lerp(Rotator) (Pawn|GetControlRotation (LocalVariables|GetPawn) :!class /Script/Engine.Pawn) (Math|Rotator|FindLookatRotation (Transformation|GetActorLocation (LocalVariables|GetPawn)) (+ (select _bsuccess (Transformation|GetWorldLocation (Variables|/Script/Engine.Character|GetMesh _ascharacter)) (Transformation|GetActorLocation (Variables|Default|GetlockedTarget))) (Variables|Default|GetTargetOffset))) (Variables|Default|GetCameraBlendRate) true)) (.yaw (Math|Rotator|Lerp(Rotator) (Pawn|GetControlRotation (LocalVariables|GetPawn) :!class /Script/Engine.Pawn) (Math|Rotator|FindLookatRotation (Transformation|GetActorLocation (LocalVariables|GetPawn)) (+ (select _bsuccess (Transformation|GetWorldLocation (Variables|/Script/Engine.Character|GetMesh _ascharacter)) (Transformation|GetActorLocation (Variables|Default|GetlockedTarget))) (Variables|Default|GetTargetOffset))) (Variables|Default|GetCameraBlendRate) true)))) ; Control rotation
              ))
            (else
              _))
          (CallFunction|ClearTargetLock)))
      (else
        (Ability|SendGameplayEvent "(TagName=\"Event.Melee.Targeting.Active\")")))
    (else
      _)))
