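; Mantis-based gameplay ability for Creative melee weapons, running attack techniques and applying their hits.
; Each technique start sets the weapon's current attack, cancels tagged abilities that would otherwise stick and spends TacticalSprint energy.
; Hits build set-by-caller damage specs from weapon data, add force feedback and knockback, and send hit-react events.
; It ends when the primary block tag appears, and relays animation notifies as gameplay events.
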(event Ability|EventActivateAbility
  ;[0 1 0.3251] Activate
  (Utilities|FlowControl|DoOnce false
    (:Completed
      (bind _outfortplayerpawn (CallFunction|GetAvatarasValidFortPlayerPawn)
        (:ValidFortPlayerPawn
          (Variables|Default|SetEnergyComp (Actor|GetComponentbyClass _outfortplayerpawn "/TacticalSprintGame/Gameplay/TacticalSprintEnergyComponent.TacticalSprintEnergyComponent_C")))
        (:AvatarCastFailed
          _))
      (|CallParentFunction)
      (bind _asynctaskproxy (Ability|Tasks|WaitGameplayTagAdd "(TagName=\"Ability.MeleeMantis.Block.Primary\")")
        (:then
          (Variables|Default|SetBlockPrimaryAttackAsyncTask _asynctaskproxy))
        (:Added
          (Ability|EndAbility))))))

(event Ability|EventOnEndAbility
  (params (bWasCancelled :type Boolean))
  ;[1 0.2088 0] End
  (|CallParentFunction bWasCancelled)
  (CallFunction|SetWeaponAttack "None")
  (GameplayTasks|EndTask (Variables|Default|GetBlockPrimaryAttackAsyncTask))
  ;[0 1 0.3251] Activate
  (Utilities|FlowControl|DoOnce false
    (:Completed
      (bind _outfortplayerpawn (CallFunction|GetAvatarasValidFortPlayerPawn)
        (:ValidFortPlayerPawn
          (Variables|Default|SetEnergyComp (Actor|GetComponentbyClass _outfortplayerpawn "/TacticalSprintGame/Gameplay/TacticalSprintEnergyComponent.TacticalSprintEnergyComponent_C")))
        (:AvatarCastFailed
          _))
      (|CallParentFunction)
      (bind _asynctaskproxy (Ability|Tasks|WaitGameplayTagAdd "(TagName=\"Ability.MeleeMantis.Block.Primary\")")
        (:then
          (Variables|Default|SetBlockPrimaryAttackAsyncTask _asynctaskproxy))
        (:Added
          (Ability|EndAbility))))))

(event Mantis|EventOnMantisTechniqueHit
  (params (TargetDataHandle :type Gameplay%20Ability%20Target%20Data%20Handle%20Structure%20%28by%20ref%29) (ApplicationTag :type Gameplay%20Tag%20Structure))
  ;[0 0.8265 1] Technique has hit something
  (CallFunction|ApplyHitEffectToTarget TargetDataHandle ApplicationTag)
  (CallFunction|SendHitReactEventToTarget TargetDataHandle))

(event Mantis|EventOnMantisAblilityTechniqueStarted
  (params (OwningTags :type Gameplay%20Tag%20Container%20Structure) (SequenceName :type Name :default "None") (StartsSequence :type Boolean) (EndsSequence :type Boolean))
  (CallFunction|SetWeaponAttack SequenceName)
  ;[1 1 1] This will be called each technique and cancel _any_ tag in the tag container. This is useful for abilities that can be overrun by this one. Since this ability doesn't follow the same sort of "ability tag" path, as it is continuously running while techniques are running, it is not enough to use the "cancel" or "block" tags in the ability class defaults. The primary example, at this time, is the jump attack ability which opens up the option to attack during the jump attack ability, which will fail to end or cancel normally and get stuck
  (bind _asfort_ability_system_component (Utilities|Casting|CastToFortAbilitySystemComponent (Ability|GetAbilitySystemComponentfromActorInfo)))
  (Tags|CancelAbilitiesWithTags _asfort_ability_system_component (Variables|Default|GetCancelAbilitiesWithTagOnTechniqueStarted))
  (CallFunction|UseAttackEnergy))

(event Ability|EventOnGenericNotifyFromAnimation
  (params (NotifyTag :type Gameplay%20Tag%20Structure))
  ;[0 0.7912 1 0.5] Mantis abilities don't seem to fully support Triggering Abilities through Montages, but they do support Generic Notifies, so this is a passthrough event to take a notify and turn it into an event trigger.
  (|CallParentFunction NotifyTag)
  (bind _playercharacter (Variables|Default|GetPlayerCharacter)
    (:then
      (Ability|SendGameplayEventtoActor _playercharacter NotifyTag))
    (:else)))

(event Custom|ApplyForceFeedback
  (if (Fortnite|DeviceSupportsForceFeedback)
    (Ability|HasAuthority
      (:False
        (bind _outfortplayerpawn (CallFunction|GetAvatarasValidFortPlayerPawn)
          (:ValidFortPlayerPawn
            (Game|Feedback|ClientPlayForceFeedback (Fortnite|GetFortPlayerControllerfromActor _outfortplayerpawn) (Variables|Default|GetForceFeedback) "None" false false false)))))))

(fn ApplyHitEffecttoTarget
  (params (TargetData :type Gameplay%20Ability%20Target%20Data%20Handle%20Structure%20%28by%20ref%29) (Application%20Tag :type Gameplay%20Tag%20Structure))
  (bind (_returnvalue_x _returnvalue_y) (Math|Vector|BreakVector (Math|Vector|Normalize (- (.ImpactPoint (Collision|BreakHitResult (Ability|EffectContext|GetHitResult (Ability|GetContextfromOwner (Variables|GetTargetData))))) (Transformation|GetActorLocation (Ability|GetAvatarActorfromActorInfo))))))
  (bind (_bblockinghit _binitialoverlap _time _distance _location _impactpoint _normal _impactnormal _physmat) (Collision|BreakHitResult (Ability|TargetData|GetHitResultfromTargetData (Variables|GetTargetData))))
  (bind _targetdata (Variables|GetTargetData))
  (bind _returnvalue (Ability|GetContextfromOwner _targetdata))
  (bind _returnvalue_2 (Ability|EffectContext|GetHitResult _returnvalue))
  (bind (_bblockinghit_2 _binitialoverlap_2 _time_2 _distance_2 _location_2 _impactpoint_2 _normal_2 _impactnormal_2 _physmat_2 _hitactor) (Collision|BreakHitResult _returnvalue_2))
  (bind _targetdata_2 (Variables|GetTargetData))
  ;[0 0.4072 0.02029]
  (Ability|ApplyGameplayEffectSpecToTarget (CallFunction|CreateDamageSpec Application%20Tag TargetData) TargetData)
  (CallFunction|ApplyForceFeedback)
  (bind _isai (CallFunction|IsAITarget (Variables|GetTargetData)))
  (bind (_wantstoknock _knockback_power _knockup_power) (CreativeMeleeWeaponInterface|GetKnockbackData(Message) (CallFunction|GetBestValidWeapon) false _isai))
  (if _wantstoknock
    (Utilities|FlowControl|Sequence
      (:then_0
        ;[1 1 1] Original use case for GA_Athena_Knockback_Parent. Keeping in case of non player pawns getting applied knockback. This fixes issue with hitching from player view but keeps functionality for everything else
        (Utilities|Casting|CastToFortPlayerPawn _hitactor
          (:CastFailed
            (Ability|ApplyGameplayEffectToTarget _targetdata_2 "/CRD_Melee/Weapons/GA/GE_Creative_Melee_Knockback_Status.GE_Creative_Melee_Knockback_Status_C")
            ; Needed this fix for FX firing on props labed as Non Damageable
            (Ability|ExecuteGameplayCueWithParamsOnOwner (Variables|Default|GetGameplay%20Cue%20Tag) (Ability|GameplayCue|MakeGameplayCueParameters :NormalizedMagnitude 0.0 :RawMagnitude 0.0 :Location _location :Normal _normal :PhysicalMaterial _physmat :GameplayEffectLevel 1 :AbilityLevel 1 :bReplicateLocationWhenUsingMinimalRepProxy false)))))
      (:then_1
        (bind _asfort_player_pawn (Utilities|Casting|CastToFortPlayerPawn _hitactor))
        (Pawn|Components|CharacterMovement|AddImpulse (Variables|/Script/Engine.Character|GetCharacterMovement _asfort_player_pawn) (Math|Vector|MakeVector (* _returnvalue_x _knockback_power) (* _returnvalue_y _knockback_power) _knockup_power) true)))))

(fn ConfigureRootMotionWarping
  (params (TechniqueName :type Name :default "None"))
  (bind (_boutallowtranslationwarp _boutallowrotationwarp _outtargetactor _boutsnapshottargetactorlocation) (|CallParentFunction TechniqueName))
  (return _boutallowtranslationwarp true _outtargetactor _boutsnapshottargetactorlocation false "0, 0, 0" false 0))

(fn CreateDamageSpec
  (params (Application%20Tag :type Gameplay%20Tag%20Structure) (TargetData :type Gameplay%20Ability%20Target%20Data%20Handle%20Structure))
  ;[1 1 1] Setting the Application tag on the weapon so we can use it to calculate knockback multiplier
  (bind _ascreative_melee_parent (Utilities|Casting|CastToCreative_Melee_Parent (Ability|GetCurrentSourceWeapon))
    (:then
      (Class|CreativeMeleeParent|SetHitDataApplicationTag _ascreative_melee_parent Application%20Tag))
    (:CastFailed
      _))
  (bind (_base_damage _environmental _aidamage _critchance _critmult) (CreativeMeleeWeaponInterface|GetHitData(Message) (Ability|GetCurrentSourceWeapon) Application%20Tag))
  (bind _isai (CallFunction|IsAITarget (Variables|GetTargetData)))
  (bind _returnvalue (Ability|GameplayEffect|AssignTagSetbyCallerMagnitude (Ability|MakeOutgoingGameplayEffectSpec (Variables|Default|GetHit%20Effect%20Class)) "(TagName=\"SetByCaller.MeleeManager.BaseDamage\")" (select _isai _aidamage _base_damage)))
  (bind _returnvalue_2 (Ability|GameplayEffect|AssignTagSetbyCallerMagnitude _returnvalue "(TagName=\"SetByCaller.MeleeManager.BaseAIDamage\")" _aidamage))
  (bind _returnvalue_3 (Ability|GameplayEffect|AssignTagSetbyCallerMagnitude _returnvalue_2 "(TagName=\"SetByCaller.MeleeManager.EnvironmentDamage\")" _environmental))
  (bind _returnvalue_4 (Ability|GameplayEffect|AssignTagSetbyCallerMagnitude _returnvalue_3 "(TagName=\"SetByCaller.MeleeManager.CritChanceModifier\")" _critchance))
  (bind _returnvalue_5 (Ability|GameplayEffect|AssignTagSetbyCallerMagnitude _returnvalue_4 "(TagName=\"SetByCaller.MeleeManager.CritModifier\")" _critmult))
  (return _returnvalue_5))

(fn IsAITarget
  (params (TargetData :type Gameplay%20Ability%20Target%20Data%20Handle%20Structure))
  (bind (_bblockinghit _binitialoverlap _time _distance _location _impactpoint _normal _impactnormal _physmat _hitactor) (Collision|BreakHitResult (Ability|EffectContext|GetHitResult (Ability|GetContextfromOwner TargetData))))
  (bind _returnvalue (GameplayTags|HasAnyMatchingGameplayTags(Message) _hitactor (Variables|Default|GetAITags)))
  (return _returnvalue))

(fn SendHitReactEventtoTarget
  (params (Target%20Data :type Gameplay%20Ability%20Target%20Data%20Handle%20Structure))
  (bind (_bblockinghit _binitialoverlap _time _distance _location _impactpoint _normal _impactnormal _physmat _hitactor) (Collision|BreakHitResult (Ability|EffectContext|GetHitResult (Ability|GetContextfromOwner Target%20Data))))
  (bind _returnvalue (Game|ConstructObjectfromClass "/CRD_Melee/HitReactions/Base/HitReactAttackData_Capsule.HitReactAttackData_Capsule_C"))
  (LocalVariables|SetHitReactDataAttackCapsule _returnvalue)
  (Variables|/CRD_Melee/HitReactions/Base/HitReactAttackData_Capsule.HitReactAttackData_Capsule_C|SetHit%20React%20Attack%20Data (Utilities|Struct|MakeHitReactAttackData (Transformation|GetActorLocation (Ability|GetAvatarActorfromActorInfo)) "/CRD_Melee/HitReactions/HitReactionType_Interrupt.HitReactionType_Interrupt" "1.000000,0.000000,0.000000") (LocalVariables|GetHitReactDataAttackCapsule))
  (Ability|SendGameplayEventtoActor _hitactor "(TagName=\"Event.Abilities.HitReact\")" (Utilities|Struct|MakeGameplayEventData "(TagName=\"\")" _hitactor _hitactor _returnvalue 0 "()" "(GameplayTags=)" "(GameplayTags=)" 0.0 (Variables|GetTarget%20Data))))

(fn SetWeaponAttack
  (params (NewAttack :type Name))
  (bind _ascreative_melee_parent (Utilities|Casting|CastToCreative_Melee_Parent (Ability|GetCurrentSourceWeapon)))
  (Class|CreativeMeleeParent|SetCurrentAttack _ascreative_melee_parent NewAttack))

(fn UseAttackEnergy
  (bind _primary_attack_cost (CreativeMeleeWeaponInterface|GetEnergyCosts(Message) (Ability|GetCurrentSourceWeapon) false))
  (bind _energycomp (Variables|Default|GetEnergyComp)
    (:then
      (Energy|UseEnergy _energycomp _primary_attack_cost true 0))
    (:else)))
