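; Primary attack gameplay ability for Creative melee weapons that handles holding input to charge the weapon's secondary attack.
; After validity and hotfix checks, parallel delays drive player slow and montage, plus weapon charge state; release at full charge grants and activates the cached secondary.
; Tag-wait tasks end the charge when blocked or retry once after tag or energy failures, through a duplicate of the activation logic.
; Airborne players instead receive a jump-and-wait-for-landing event.
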(event Custom|ToggleChargedUpPrep
  (params (ChargedUp :type Boolean))
  (bind _output_get (Variables|Runtime|SetChargedUp ChargedUp))
  (if _output_get
    (bind _returnvalue (Ability|ApplyGameplayEffectToOwner (Variables|Ability|GetGEChargingMovementSpeed)))
    (Variables|Tags|SetChargingMovementSpeedHandle _returnvalue)
    (else
      (Ability|RemoveGameplayEffectFromOwnerWithHandle (Variables|Tags|GetChargingMovementSpeedHandle)))))

(event Custom|TryPlayChargeAnim
  (bind (_valid _softobj) (CallFunction|GetChargeAnimation))
  (if _valid
    (bind _object (Utilities|AsyncLoadAsset _softobj)
      (:Completed
        (bind _asanim_montage (Utilities|Casting|CastToAnimMontage _object))
        (Ability|Tasks|PlayMontageAndWaitNotify "None" _asanim_montage)))))

(event Custom|Retry%20After%20Fail
  (bind _outfortplayerpawn (CallFunction|GetAvatarasValidFortPlayerPawn)
    (:ValidFortPlayerPawn
      (CallFunction|CheckEnergyforPrimary)
      (if (not (AI|Components|NavMovement|IsFalling (Variables|/Script/Engine.Character|GetCharacterMovement _outfortplayerpawn)))
        (bind (_valid _current_weapon) (CallFunction|CacheSecondaryAbility))
        (if _valid
          (bind _result (CallFunction|CheckCanChargeSecondary))
          (if _result
            (bind _ascreative_melee_parent (Utilities|Casting|CastToCreative_Melee_Parent _current_weapon)
              (:then
                (bind _output_get (Variables|Runtime|SetCurrentWeapon _ascreative_melee_parent))
                (Variables|Runtime|SetTotal%20Charge%20Time (Class|CreativeMeleeParent|GetChargeDetails _output_get))
                (bind _output_get_2 (Variables|Const|SetInitial%20Charge%20Delay (.Initial%20Charge%20Delay (Class|CreativeMeleeParent|GetChargeDetails _output_get))))
                (Utilities|FlowControl|Delay _output_get_2)
                (Utilities|CanActivateSecondary
                  (:Yes
                    (bind _asynctaskproxy (Ability|Tasks|WaitGameplayTagAdd (Variables|Tags|GetPreventSecondaryTag))
                      (:then
                        (Variables|Task|SetPreventSecondaryTask _asynctaskproxy)
                        (Class|CreativeMeleeParent|SetChargeState (Variables|Runtime|GetCurrentWeapon) "NewEnumerator2" :!class /CRD_Melee/Weapons/Creative_Melee_Parent.Creative_Melee_Parent_C)
                        (Utilities|FlowControl|Sequence
                          (:then_0
                            ;[0 1 0.3155] Player's Charge State (as opposed to weapon). This covers animation and movement.
                            (Utilities|FlowControl|Delay (Math|Float|Min(Float) (Math|Float|SafeDivide (Variables|Runtime|GetTotal%20Charge%20Time) (Variables|Const|GetCharge%20Up%20Prep%20Divisor)) (Variables|Const|GetChargeUpTogglePrepDelay) (- (Variables|Runtime|GetTotal%20Charge%20Time) (Variables|Const|GetInitial%20Charge%20Delay))))
                            (CallFunction|ToggleChargedUpPrep true)
                            (CallFunction|TryPlayChargeAnim)
                            (bind _asynctaskproxy_2 (Ability|Tasks|WaitInputRelease false)
                              (:then
                                (Variables|Task|SetChargingUpInputListen _asynctaskproxy_2))
                              (:OnRelease
                                (Ability|EndAbility))))
                          (:then_1
                            ;[0 1 0.3098] this delay will tell us when we are fully charged
                            (Utilities|FlowControl|Delay (- (Variables|Runtime|GetTotal%20Charge%20Time) (Variables|Const|GetInitial%20Charge%20Delay)))
                            (GameplayTasks|EndTask (Variables|Task|GetPreventSecondaryTask))
                            (Class|CreativeMeleeParent|SetChargeState (Variables|Runtime|GetCurrentWeapon) "NewEnumerator3" :!class /CRD_Melee/Weapons/Creative_Melee_Parent.Creative_Melee_Parent_C)
                            (Ability|Tasks|WaitInputRelease false
                              (:OnRelease
                                (CallFunction|CullFlyingonChargedAttack)
                                (CallFunction|ActivateSecondaryAbility)
                                (Utilities|FlowControl|Delay 0.5)
                                (Ability|EndAbility))))))
                      (:Added
                        (Ability|EndAbility))))
                  (:No
                    ;[0 0.6697 1 0.25] If we can't currently charge secondary, wait until the secondary stops, then retry.
                    ; If this happens multiple times in the same ability, just end.
                    (Utilities|DoOnceThenDo false
                      (:DoOnce
                        (bind _asynctaskproxy_3 (Ability|Tasks|WaitGameplayTagRemove (Variables|Tags|GetPreventSecondaryTag) 0 true)
                          (:then
                            (Variables|Task|SetSecondaryRemovalAsync _asynctaskproxy_3))
                          (:Removed
                            (CallFunction|RetryAfterFail))))
                      (:ThenDo
                        (Ability|EndAbility))))
                  (:NotEnoughEnergy
                    ;[0 0.6697 1 0.25] If we can't currently charge secondary due to lack of energy, wait until the secondary stops, then retry.
                    ; If this happens multiple times in the same ability, just end.
                    (Utilities|DoOnceThenDo false
                      (:DoOnce
                        (bind _asynctaskproxy_4 (Ability|Tasks|WaitGameplayTagRemove (Variables|Default|GetNotEnoughEnergyTagSecondary) 0 true)
                          (:then
                            (Variables|Task|SetSecondaryEnergyAsync _asynctaskproxy_4))
                          (:Removed
                            (CallFunction|RetryAfterFail))))
                      (:ThenDo
                        (Ability|EndAbility))))))
              (:CastFailed
                _))
            (else
              _))
          (Ability|EndAbility)
          (else
            _))
        (else
          (Ability|SendGameplayEventtoActor _outfortplayerpawn (Variables|Tags|GetEventJumpAndWaitForLanding) (Utilities|Struct|MakeGameplayEventData (Variables|Tags|GetEventJumpAndWaitForLanding) (Ability|GetAvatarActorfromActorInfo) 0 0 0 "()" "(GameplayTags=)" "(GameplayTags=)" (Variables|Const|GetJumpMagnitude))))))
    (:AvatarCastFailed
      (Ability|EndAbility))))

(event Ability|EventActivateAbility
  (Variables|Runtime|SetChargedUp false)
  (bind _asynctaskproxy (Ability|Tasks|WaitInputRelease false)
    (:then
      (Variables|Task|SetWaitInputReleasedAsyncTask _asynctaskproxy)
      (bind _outfortplayerpawn (CallFunction|GetAvatarasValidFortPlayerPawn)
        (:ValidFortPlayerPawn
          (CallFunction|CheckEnergyforPrimary)
          (if (not (AI|Components|NavMovement|IsFalling (Variables|/Script/Engine.Character|GetCharacterMovement _outfortplayerpawn)))
            (bind (_valid _current_weapon) (CallFunction|CacheSecondaryAbility))
            (if _valid
              (bind _result (CallFunction|CheckCanChargeSecondary))
              (if _result
                (bind _ascreative_melee_parent (Utilities|Casting|CastToCreative_Melee_Parent _current_weapon)
                  (:then
                    (bind _output_get (Variables|Runtime|SetCurrentWeapon _ascreative_melee_parent))
                    (Variables|Runtime|SetTotal%20Charge%20Time (Class|CreativeMeleeParent|GetChargeDetails _output_get))
                    (bind _output_get_2 (Variables|Const|SetInitial%20Charge%20Delay (.Initial%20Charge%20Delay (Class|CreativeMeleeParent|GetChargeDetails _output_get))))
                    (Utilities|FlowControl|Delay _output_get_2)
                    (Utilities|CanActivateSecondary
                      (:Yes
                        (bind _asynctaskproxy_2 (Ability|Tasks|WaitGameplayTagAdd (Variables|Tags|GetPreventSecondaryTag))
                          (:then
                            (Variables|Task|SetPreventSecondaryTask _asynctaskproxy_2)
                            (Class|CreativeMeleeParent|SetChargeState (Variables|Runtime|GetCurrentWeapon) "NewEnumerator2" :!class /CRD_Melee/Weapons/Creative_Melee_Parent.Creative_Melee_Parent_C)
                            (Utilities|FlowControl|Sequence
                              (:then_0
                                ;[0 1 0.3155] Player's Charge State (as opposed to weapon). This covers animation and movement.
                                (Utilities|FlowControl|Delay (Math|Float|Min(Float) (Math|Float|SafeDivide (Variables|Runtime|GetTotal%20Charge%20Time) (Variables|Const|GetCharge%20Up%20Prep%20Divisor)) (Variables|Const|GetChargeUpTogglePrepDelay) (- (Variables|Runtime|GetTotal%20Charge%20Time) (Variables|Const|GetInitial%20Charge%20Delay))))
                                (CallFunction|ToggleChargedUpPrep true)
                                (CallFunction|TryPlayChargeAnim)
                                (bind _asynctaskproxy_3 (Ability|Tasks|WaitInputRelease false)
                                  (:then
                                    (Variables|Task|SetChargingUpInputListen _asynctaskproxy_3))
                                  (:OnRelease
                                    (Ability|EndAbility))))
                              (:then_1
                                ;[0 1 0.3098] this delay will tell us when we are fully charged
                                (Utilities|FlowControl|Delay (- (Variables|Runtime|GetTotal%20Charge%20Time) (Variables|Const|GetInitial%20Charge%20Delay)))
                                (GameplayTasks|EndTask (Variables|Task|GetPreventSecondaryTask))
                                (Class|CreativeMeleeParent|SetChargeState (Variables|Runtime|GetCurrentWeapon) "NewEnumerator3" :!class /CRD_Melee/Weapons/Creative_Melee_Parent.Creative_Melee_Parent_C)
                                (Ability|Tasks|WaitInputRelease false
                                  (:OnRelease
                                    (CallFunction|CullFlyingonChargedAttack)
                                    (CallFunction|ActivateSecondaryAbility)
                                    (Utilities|FlowControl|Delay 0.5)
                                    (Ability|EndAbility))))))
                          (:Added
                            (Ability|EndAbility))))
                      (:No
                        ;[0 0.6697 1 0.25] If we can't currently charge secondary, wait until the secondary stops, then retry.
                        ; If this happens multiple times in the same ability, just end.
                        (Utilities|DoOnceThenDo false
                          (:DoOnce
                            (bind _asynctaskproxy_4 (Ability|Tasks|WaitGameplayTagRemove (Variables|Tags|GetPreventSecondaryTag) 0 true)
                              (:then
                                (Variables|Task|SetSecondaryRemovalAsync _asynctaskproxy_4))
                              (:Removed
                                (CallFunction|RetryAfterFail))))
                          (:ThenDo
                            (Ability|EndAbility))))
                      (:NotEnoughEnergy
                        ;[0 0.6697 1 0.25] If we can't currently charge secondary due to lack of energy, wait until the secondary stops, then retry.
                        ; If this happens multiple times in the same ability, just end.
                        (Utilities|DoOnceThenDo false
                          (:DoOnce
                            (bind _asynctaskproxy_5 (Ability|Tasks|WaitGameplayTagRemove (Variables|Default|GetNotEnoughEnergyTagSecondary) 0 true)
                              (:then
                                (Variables|Task|SetSecondaryEnergyAsync _asynctaskproxy_5))
                              (:Removed
                                (CallFunction|RetryAfterFail))))
                          (:ThenDo
                            (Ability|EndAbility))))))
                  (:CastFailed
                    _))
                (else
                  _))
              (Ability|EndAbility)
              (else
                _))
            (else
              (Ability|SendGameplayEventtoActor _outfortplayerpawn (Variables|Tags|GetEventJumpAndWaitForLanding) (Utilities|Struct|MakeGameplayEventData (Variables|Tags|GetEventJumpAndWaitForLanding) (Ability|GetAvatarActorfromActorInfo) 0 0 0 "()" "(GameplayTags=)" "(GameplayTags=)" (Variables|Const|GetJumpMagnitude))))))
        (:AvatarCastFailed
          _)))
    (:OnRelease
      (if (not (Variables|Runtime|GetChargedUp))
        _)))
  (Ability|EndAbility))

(event Custom|ResetSecondaryCheck
  (Utilities|FlowControl|Sequence
    (:then_0
      ;[0 0.6697 1 0.25] If we can't currently charge secondary, wait until the secondary stops, then retry.
      ; If this happens multiple times in the same ability, just end.
      (Utilities|DoOnceThenDo false
        (:DoOnce
          (bind _asynctaskproxy (Ability|Tasks|WaitGameplayTagRemove (Variables|Tags|GetPreventSecondaryTag) 0 true)
            (:then
              (Variables|Task|SetSecondaryRemovalAsync _asynctaskproxy))
            (:Removed
              (CallFunction|RetryAfterFail))))
        (:ThenDo
          (Ability|EndAbility))))
    (:then_1
      ;[0 0.6697 1 0.25] If we can't currently charge secondary due to lack of energy, wait until the secondary stops, then retry.
      ; If this happens multiple times in the same ability, just end.
      (Utilities|DoOnceThenDo false
        (:DoOnce
          (bind _asynctaskproxy_2 (Ability|Tasks|WaitGameplayTagRemove (Variables|Default|GetNotEnoughEnergyTagSecondary) 0 true)
            (:then
              (Variables|Task|SetSecondaryEnergyAsync _asynctaskproxy_2))
            (:Removed
              (CallFunction|RetryAfterFail))))
        (:ThenDo
          (Ability|EndAbility))))))

(event Ability|EventOnEndAbility
  (params (bWasCancelled :type Boolean))
  (CallFunction|ResetSecondaryCheck)
  (GameplayTasks|EndTask (Variables|Task|GetWaitInputReleasedAsyncTask))
  ;[1 0.1245 0 0.25] End Ability. Cancel Charge Effects if they're currenly active.
  (CallFunction|ToggleChargedUpPrep false)
  (Class|CreativeMeleeParent|SetChargeState (Variables|Runtime|GetCurrentWeapon) "NewEnumerator1" :!class /CRD_Melee/Weapons/Creative_Melee_Parent.Creative_Melee_Parent_C))

(fn ActivateSecondaryAbility
  ;[0 0.8381 1 0.25] Fires the ability that the weapon is setup to use (cached)
  (bind _outfortplayerpawn (CallFunction|GetAvatarasValidFortPlayerPawn)
    (:ValidFortPlayerPawn
      (Ability|Advanced|GrantandActivateAbilityonSelf (Ability|GetAbilitySystemComponent _outfortplayerpawn) (Variables|Ability|GetSecondary%20Ability)))
    (:AvatarCastFailed
      (Ability|EndAbility))))

(fn CacheSecondaryAbility
  (bind _returnvalue (Ability|GetCurrentSourceWeapon))
  (Utilities|IsValid _returnvalue
    (:Is%20Valid
      (bind _abilitytag (CreativeMeleeWeaponInterface|GetSecondaryAbility(Message) _returnvalue))
      (bind _output_get (Variables|Ability|SetSecondary%20Ability _abilitytag))
      (return (Utilities|IsValidClass _output_get) _returnvalue))
    (:Is%20Not%20Valid
      (return false 0))))

(fn CheckCanChargeSecondary
  (bind _outfortplayerpawn (CallFunction|GetAvatarasValidFortPlayerPawn)
    (:ValidFortPlayerPawn
      (bind _returnvalue (CallFunction|CurrentSecondaryAttackIsEnabled))
      (return (and _returnvalue (not (GameplayTags|HasAnyMatchingGameplayTags _outfortplayerpawn (Variables|Default|GetPreventChargeTags))))))
    (:AvatarCastFailed
      (return false))))

(fn CheckEnergyForPrimary
  (Networking|CanExecuteCosmeticEvents
    (:True
      (bind _outfortplayerpawn (CallFunction|GetAvatarasValidFortPlayerPawn)
        (:ValidFortPlayerPawn
          (if (GameplayTags|HasMatchingGameplayTag _outfortplayerpawn "(TagName=\"Ability.Melee.NotEnoughEnergy.Primary\")")
            (FailedAction|UnabletoPerformAction _outfortplayerpawn false (Variables|Default|GetFail%20Attack%20Reason))
            (GameplayCue|ExecuteGameplayCue _outfortplayerpawn (Variables|Default|GetGCN_NotEnoughEnergy) (Class|AthenaFunctionLibrary|MakeSimpleFortEffectContext _outfortplayerpawn))))))))

(fn CullFlyingOnChargedAttack
  ;[1 0.8223 0] Check to see if the player is flying using creative flying. Toggle it off here. Mainfly for use when performing heavy attacks. May have to expand to filter out what type of heavy attack down the line.
  (bind _outfortplayerpawn (CallFunction|GetAvatarasValidFortPlayerPawn)
    (:ValidFortPlayerPawn
      (Utilities|IsValid (Ability|GetAbilitySystemComponent _outfortplayerpawn)
        (:Is%20Valid
          (if (GameplayTags|HasAnyMatchingGameplayTags (Ability|GetAbilitySystemComponent _outfortplayerpawn) (GameplayTags|MakeGameplayTagContainerfromTag (Variables|Default|GetCreativeFlyingTag)))
            (bind _asfort_player_controller_gameplay (Utilities|Casting|CastToFortPlayerControllerGameplay (Pawn|GetController _outfortplayerpawn)))
            (Controller|Flying|CreativeToggleFly _asfort_player_controller_gameplay false)))))))

(fn CurrentSecondaryAttackIsEnabled
  (bind _secondaryattackhotfixmap (Variables|Default|GetSecondaryAttackHotfixMap))
  (if (Utilities|IsValidClass (Variables|Ability|GetSecondary%20Ability))
    (for _index (range (Utilities|Map|Length _secondaryattackhotfixmap))
      (bind _keys (Utilities|Map|Keys _secondaryattackhotfixmap))
      (if (and (ScalableFloat|GetValueasBool (Utilities|Map|Find (Variables|Default|GetSecondaryAttackHotfixMap) (Utilities|Array|Get(acopy) _keys _index))) (Utilities|ClassIsChildOf (Variables|Ability|GetSecondary%20Ability) (Utilities|Array|Get(acopy) _keys _index)))
        (return true)
        (else
          _)))
    (return false)
    (else
      (return false))))

(fn GetChargeAnimation
  (bind _chargeanimpairings (Variables|Animation|GetChargeAnimPairings))
  (bind _creativeweapon (LocalVariables|GetCreativeWeapon))
  (bind _currentattackname (Variables|/CRD_Melee/Weapons/Creative_Melee_Parent.Creative_Melee_Parent_C|GetCurrentAttackName _creativeweapon))
  (bind _outfortplayerpawn (CallFunction|GetAvatarasValidFortPlayerPawn)
    (:ValidFortPlayerPawn
      (bind _ascreative_melee_parent (Utilities|Casting|CastToCreative_Melee_Parent (Variables|/Script/FortniteGame.FortPawn|GetCurrentWeapon _outfortplayerpawn))
        (:then
          (bind _output_get (LocalVariables|SetCreativeWeapon _ascreative_melee_parent))
          (bind _array_element (Utilities|Array|ForEachLoopwithBreak (GameplayTags|BreakGameplayTagContainer (Tags|GetWeaponDataGameplayTags _output_get)))
            (:LoopBody
              (if (.ReturnValue (Utilities|Map|Find _chargeanimpairings _array_element))
                (if (.ReturnValue (Utilities|Map|Find (Utilities|Struct|BreakWeaponTagChargeAnimationPairing (Utilities|Map|Find _chargeanimpairings _array_element)) _currentattackname))
                  (LocalVariables|SetFound (Utilities|Map|Find (Utilities|Struct|BreakWeaponTagChargeAnimationPairing (Utilities|Map|Find _chargeanimpairings _array_element)) _currentattackname)))))
            (:Completed
              _))
          (if (Utilities|IsValidSoftObjectReference (LocalVariables|GetFound))
            (return true (LocalVariables|GetFound))
            (else
              (return false 0))))
        (:CastFailed
          _)))
    (:AvatarCastFailed
      _))
  (return false 0))

(fn GetChargeVFX
  (params (Fort%20Player :type Fort%20Player%20Pawn%20Object%20Reference))
  (bind _currentweapon (Variables|/Script/FortniteGame.FortPawn|GetCurrentWeapon Fort%20Player))
  (bind _vfx_dictionary (Variables|VFX|GetVFX%20Dictionary))
  ;[0.074 1 0] NOT using the get weapon data gameplay tags because it will append _other_ tags to the list and return them. In this case, it meant that the "WeaponType.Spear", "WeaponType.Sword", and "WeaponType.Hammer" would append to the list, which would result in whatever one is found first in the map would play. For now, directly access the Gameplay Tags from the Weapon Data.
  (Utilities|IsValid _currentweapon
    (:Is%20Valid
      (bind _output_get (LocalVariables|SetGameplay%20Tags (GameplayTags|GetOwnedGameplayTags (Data|GetWeaponData _currentweapon) :!class /Script/GameplayTags.GameplayTagAssetInterface)))
      (bind _array_element (Utilities|Array|ForEachLoopwithBreak (GameplayTags|BreakGameplayTagContainer _output_get))
        (:LoopBody
          (if (.ReturnValue (Utilities|Map|Find _vfx_dictionary _array_element))
            (Utilities|FlowControl|Sequence
              (:then_0
                _)
              (:then_1
                _))
            (return (Utilities|Map|Find _vfx_dictionary _array_element))))))))
